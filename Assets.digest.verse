# Copyright Epic Games, Inc. All Rights Reserved.
#################################################
# Generated Digest of Verse API
# DO NOT modify this manually!
# Generated from build: ++Fortnite+Release-36.30-CL-44367537
#################################################


using {/Verse.org/Native}
using {/Verse.org/SceneGraph}
using {/Verse.org/Simulation}
using {/Verse.org/Colors}
using {/Verse.org/Assets}
# Module import path: /playkigai.dig@fortnite.com/Minigame_TankBrothers01/CustomImages
CustomImages := module:
    Border<scoped {/playkigai.dig@fortnite.com/Minigame_TankBrothers01}>:texture = external {}

    Heart<scoped {/playkigai.dig@fortnite.com/Minigame_TankBrothers01}>:texture = external {}

# Module import path: /playkigai.dig@fortnite.com/Minigame_TankBrothers01/Materials
Materials := module:
    tank_material<scoped {/playkigai.dig@fortnite.com/Minigame_TankBrothers01}> := class<final><scoped {/playkigai.dig@fortnite.com/Minigame_TankBrothers01}>(material):
        @editable
        var BaseColor<public>:color = external {}

# Module import path: /playkigai.dig@fortnite.com/Minigame_TankBrothers01/Meshes
Meshes := module:
    Explosion<scoped {/playkigai.dig@fortnite.com/Minigame_TankBrothers01}> := class<final><scoped {/playkigai.dig@fortnite.com/Minigame_TankBrothers01}>(mesh_component):
        @editable
        var Material_0<public>:material = external {}

    Explosion_asset<scoped {/playkigai.dig@fortnite.com/Minigame_TankBrothers01}>:mesh = external {}

    Projectile<scoped {/playkigai.dig@fortnite.com/Minigame_TankBrothers01}> := class<final><scoped {/playkigai.dig@fortnite.com/Minigame_TankBrothers01}>(mesh_component):
        @editable
        var Material_0<public>:material = external {}

    Projectile_asset<scoped {/playkigai.dig@fortnite.com/Minigame_TankBrothers01}>:mesh = external {}

    Tank<scoped {/playkigai.dig@fortnite.com/Minigame_TankBrothers01}> := class<final><scoped {/playkigai.dig@fortnite.com/Minigame_TankBrothers01}>(mesh_component):
        @editable
        var Material<public>:material = external {}

    Tank_asset<scoped {/playkigai.dig@fortnite.com/Minigame_TankBrothers01}>:mesh = external {}

    Turret<scoped {/playkigai.dig@fortnite.com/Minigame_TankBrothers01}> := class<final><scoped {/playkigai.dig@fortnite.com/Minigame_TankBrothers01}>(mesh_component):
        @editable
        var Material<public>:material = external {}

    Turret_asset<scoped {/playkigai.dig@fortnite.com/Minigame_TankBrothers01}>:mesh = external {}

# Module import path: /playkigai.dig@fortnite.com/Minigame_TankBrothers01/Prefabs
Prefabs := module:
    @import_as("/Minigame_TankBrothers01/Prefabs/explosion.explosion_C")
    explosion<scoped {/playkigai.dig@fortnite.com/Minigame_TankBrothers01}> := class<final>(entity):


    explosion_asset<scoped {/playkigai.dig@fortnite.com/Minigame_TankBrothers01}>:entity_prefab = external {}

    @import_as("/Minigame_TankBrothers01/Prefabs/game_manager.game_manager_C")
    game_manager<scoped {/playkigai.dig@fortnite.com/Minigame_TankBrothers01}> := class<final>(entity):


    game_manager_asset<scoped {/playkigai.dig@fortnite.com/Minigame_TankBrothers01}>:entity_prefab = external {}

    @import_as("/Minigame_TankBrothers01/Prefabs/projectile.projectile_C")
    projectile<scoped {/playkigai.dig@fortnite.com/Minigame_TankBrothers01}> := class<final>(entity):


    projectile_asset<scoped {/playkigai.dig@fortnite.com/Minigame_TankBrothers01}>:entity_prefab = external {}

    @import_as("/Minigame_TankBrothers01/Prefabs/tank.tank_C")
    tank<scoped {/playkigai.dig@fortnite.com/Minigame_TankBrothers01}> := class<final>(entity):


    tank_asset<scoped {/playkigai.dig@fortnite.com/Minigame_TankBrothers01}>:entity_prefab = external {}

    @import_as("/Minigame_TankBrothers01/Prefabs/turret.turret_C")
    turret<scoped {/playkigai.dig@fortnite.com/Minigame_TankBrothers01}> := class<final>(entity):


    turret_asset<scoped {/playkigai.dig@fortnite.com/Minigame_TankBrothers01}>:entity_prefab = external {}