<#
@UUID: 2f2c663a-28d4-433a-8a87-8d02aec942fc
@Task: Custom explosion effect that interacts with entities
@Description: Module which handles the "explosion" effect when a `projectile` collides with other objects.
@Tags: Explosion, Instantiation, Spawning, Damage, Collision
@Export: true
#>
<#
Entity Prefab setup for `explosion` entity prefab:
    `explosion` is a top-level entity prefab that gets spawned whenever a `projectile` entity overlaps with another object.

    All components listed in the entity setup hierarchy *must* match to ensure `explosion` acts as expected.

    - explosion
        - transform_component
        - mesh_component
            NOTE:
                The mesh associated with mesh_component *must* have collision.
                Its collision *must* allow queries (e.g. `OverlapAllDynamic` collision preset).
                    "Collision Complexity" should be set to `Project Default`.
        - keyframed_movement_component
            NOTE:
                keyframed_movement_component is used to expand the explosion mesh
                    and allow the `explosion` to overlap other entities.
        - explosion_component
#>
using { /Verse.org }
using { /Verse.org/Native }
using { /Verse.org/SceneGraph }
using { /Verse.org/Simulation }
using { /Verse.org/SceneGraph/KeyframedMovement }
using { /Verse.org/SpatialMath }
using { Prefabs }


explosion_component := class<final_super>(component):

    @editable
    ExplosionRadius: float = 20.0
    @editable
    ExplosionDuration: float = 0.5

    var MaybeOwnerAgent: ?agent = false
    var TankHits: []tank = array{}


    OnBeginSimulation<override>():void =
        (super:)OnBeginSimulation()

    OnSimulate<override>()<suspends>:void=
        if(MovementComp := Entity.GetComponent[keyframed_movement_component]):
            CurrentTransform := Entity.GetGlobalTransform()
            KeyFrame := keyframed_movement_delta:
                Transform := transform:
                    Translation := vector3{}
                    Rotation := IdentityRotation()
                    Scale := CurrentTransform.Scale * ExplosionRadius
                Duration := ExplosionDuration

            MovementComp.SetKeyframes(array. KeyFrame, oneshot_keyframed_movement_playback_mode{})
            MovementComp.Play()

            race:
                MovementComp.FinishedEvent.Await()

                loop:
                    HandleCollisions()

                    Sleep(0.0)

            Entity.RemoveFromParent()    


    HandleCollisions():void=
        for:
            Hit : Entity.FindOverlapHits()
            HitTank := tank[Hit.TargetComponent.Entity]
            TankComp := HitTank.GetComponent[tank_controller_component]
            TankComp.MaybePilot <> MaybeOwnerAgent
            not TankCached(HitTank)?
        do:
            HitTank.Damage(1)
            set TankHits += array. HitTank


    TankCached(Tank:tank)<transacts>:logic=
        for(CachedTank : TankHits, Tank = CachedTank):
            return true

        false
