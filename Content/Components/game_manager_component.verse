<#
@UUID: 1daf52ab-da59-410e-90a8-ac959972017c
@Task: Game Manager
@Description: Manages gamestate, player spawning, starting and stopping rounds, and queue.
@Tags: Gamestate, Game Manager, Custom Input, Utility Functions
@Export: true
#>
<#
Entity Prefab setup for `game_manager` entity prefab:
    `game_manager` is the core logic handler for the Tank Brothers minigmae. It handles
        players joining/leaving queue, spawning tanks, round state, remaining active players,
        declaring a winnner, etc.

    All components listed in the entity setup hierarchy *must* match to ensure `game_manager` acts as expected.

    - game_manager
        - transform_component
        - mesh_component
            NOTE:
                The mesh associated with mesh_component *must* have collision.
                Its collision *must* allow queries (e.g. `PhysicsActor` collision preset).
                    "Collision Complexity" should be set to `Project Default`.
                Collision is required to allow players to interact with the game_manager prefab
                    in the scene in order to join/leave queue.
        - game_manager_component
#>
using { /Verse.org }
using { /Verse.org/Native }
using { /Verse.org/SceneGraph }
using { /Verse.org/Simulation }
using { /Fortnite.com/Devices }
using { /Verse.org/SpatialMath }
using { /Verse.org/Colors }
using { /Verse.org/Random }
using { Prefabs }

    game_states := enum:
        AwaitingStart,
        InProgress

    game_manager_component := class<final_super>(interactable_component):

        @editable
        TankCameras: []gameplay_camera_fixed_point_device = array{}
    
        @editable
        FireTankInput: input_trigger_device = input_trigger_device{}

        @editable
        MoveFwdInput: input_trigger_device = input_trigger_device{}
        @editable
        MoveBwdInput: input_trigger_device = input_trigger_device{}
        @editable
        MoveLeftInput: input_trigger_device = input_trigger_device{}
        @editable
        MoveRightInput: input_trigger_device = input_trigger_device{}

        @editable
        QueueTime: int = 3
        @editable
        MaxPlayerCount: int = 4
        @editable
        MinPlayerCount: int = 2
        @editable
        TankSpawnRadius: float = 1500.0

        GameStarted: event() = event(){}
        GameEnded: event(?agent) = event(?agent){}
        RemovedFromScene: event() = event(){}
        TankDestroyed: event(tank) = event(tank){}

        var GameState: game_states
        var AllowSpawnTanks: logic = false
        var QueuedPlayers: []agent = array{}
        var SpawnedTanks: []tank = array{}
        var QueueTimer: int = 0


        OnBeginSimulation<override>():void=
            SucceededEvent.Subscribe(OnInteractedWith)       
            
            set AllowSpawnTanks = true


        OnSimulate<override>()<suspends>:void=
            race:
                RemovedFromScene.Await()

                sync:
                    loop:
                        if(GameState = game_states.AwaitingStart):
                            SetQueueTimer()

                            Sleep(1.0)

                        if(QueueTimerExpired[]):
                            SpawnTanks()
                            set QueuedPlayers = array{}

                            GameStarted.Signal()
                            set GameState = game_states.InProgress
                            
                            MaybeWinner := GameEnded.Await()

                            for(Tank : SpawnedTanks):
                                Tank.Destroy()
                            if(Winner := MaybeWinner?):
                                Print("{Winner} wins!")
                            else:
                                Print("Game over.")

                            set GameState = game_states.AwaitingStart

                        Sleep(0.0)
                    
                    loop:
                        DestroyedTank := TankDestroyed.Await()

                        var NewArray: []tank = array{}
                        for(Tank : SpawnedTanks, Tank <> DestroyedTank):
                            set NewArray += array{Tank}
                        set SpawnedTanks = NewArray

                        if(SpawnedTanks.Length <= 1):
                            var MaybeWinner: ?agent = false
                            if(TankComp := SpawnedTanks[0].GetComponent[tank_controller_component], Pilot := TankComp.MaybePilot?):
                                set MaybeWinner = option. Pilot

                            GameEnded.Signal(MaybeWinner)                                

                        Sleep(0.0)

            
        OnEndSimulation<override>():void=
            set AllowSpawnTanks = false
            RemovedFromScene.Signal()


        SetQueueTimer():void=
            if(PlayerCountValid[]):
                set QueueTimer -= 1
                Print("Game starts in {QueueTimer}", ?Duration := 1.0)
            else if(not QueueTimer = QueueTime):
                set QueueTimer = QueueTime


        SpawnTanks():void=
            for:
                X -> Camera : TankCameras
                Player := QueuedPlayers[X]
                SimRoot := Entity.GetSimulationEntity[]
                Tank := tank{}
                Turret := turret{}
                TankComp := Tank.GetComponent[tank_controller_component]
            do:
                SimRoot.AddEntities(array. Tank)
                SimRoot.AddEntities(array. Turret)
                set SpawnedTanks += array. Tank
                set TankComp.MaybeTurret = option. Turret
                set TankComp.MaybeGameManager = option. Self
                set TankComp.MaybeCamera = option. Camera

                Tank.ChangeColor(GetRandomColor())
                Tank.EnterAsPilot(Player)

            for(X -> Tank : SpawnedTanks, Angle := ((1.0 * X) / (1.0 * SpawnedTanks.Length)) * 6.28):
                NewLoc := vector3:
                        Left := Cos(Angle) * TankSpawnRadius
                        Up := 0.0
                        Forward := Sin(Angle) * TankSpawnRadius
                NewTransform := transform:
                    Translation := NewLoc
                    Rotation := MakeShortestRotationBetween(Tank.GetGlobalTransform().Rotation.GetForwardAxis(), Entity.GetGlobalTransform().Translation - NewLoc)

                Tank.SetGlobalTransform(NewTransform)


        OnInteractedWith(Agent:agent):void=
            if(not GameState = game_states.AwaitingStart):
                GameEnded.Signal(false)
                return

            if(QueuedPlayers.Find[Agent], var NewArray: []agent = array{}):
                for(Player : QueuedPlayers, Player <> Agent):
                    set NewArray += array. Player      
                set QueuedPlayers = NewArray
                Print("{Agent} left queue ({QueuedPlayers.Length} / {MinPlayerCount})")

            else if(QueuedPlayers.Length < MaxPlayerCount):
                set QueuedPlayers += array. Agent
                Print("{Agent} joined queue ({QueuedPlayers.Length} / {MinPlayerCount})")



        InteractMessage<override>(Agent:agent)<reads><decides>:message= 
            if. GameState = game_states.InProgress then GameInProgressText
            else if. QueuedPlayers.Length >= MaxPlayerCount then QueueFullText
            else if. QueuedPlayers.Find[Agent] then LeaveQueueText 
            else JoinQueueText


        GameInProgressText<localizes>:message= "Game in progress"
        QueueFullText<localizes>:message= "Queue full"      
        LeaveQueueText<localizes>:message= "Leave Queue"
        JoinQueueText<localizes>:message= "Join Queue"

        AgentNameText<localizes>(Agent:agent):message= "{Agent}"
        ToString(Agent:agent):string= Localize(AgentNameText(Agent))       


    (GameManager:game_manager_component).StartGame():void=
        if(not GameManager.GameState = game_states.AwaitingStart):
            return

        GameManager.GameStarted.Signal()


    (GameManager:game_manager_component).PlayerCountValid()<transacts><decides>:void=
        PlayerCount := GameManager.QueuedPlayers.Length
        PlayerCount >= GameManager.MinPlayerCount and PlayerCount <= GameManager.MaxPlayerCount


    (GameManager:game_manager_component).QueueTimerExpired()<transacts><decides>:void= GameManager.QueueTimer <= 0


    
    # === Utility ===
    input_result := struct:
        FirePressed: logic = false
        MoveForward: logic = false
        MoveBackward: logic = false
        MoveLeft: logic = false
        MoveRight: logic = false


    (GameManager:game_manager_component).GetInput(Agent:agent)<transacts>:input_result=
        input_result:
            FirePressed := if. GameManager.FireTankInput.IsHeld[Agent] then true else false
            MoveForward := if. GameManager.MoveFwdInput.IsHeld[Agent] then true else false
            MoveBackward := if. GameManager.MoveBwdInput.IsHeld[Agent] then true else false
            MoveLeft := if. GameManager.MoveLeftInput.IsHeld[Agent] then true else false
            MoveRight := if. GameManager.MoveRightInput.IsHeld[Agent] then true else false


    (InputResult:input_result).GetMovementDirection():vector3=
        var LeftAxis: float = 0.0        
        if. InputResult.MoveRight? then set LeftAxis += 1.0
        if. InputResult.MoveLeft? then set LeftAxis -= 1.0

        var ForwardAxis: float = 0.0
        if. InputResult.MoveForward? then set ForwardAxis += 1.0
        if. InputResult.MoveBackward? then set ForwardAxis -= 1.0

        Direction := vector3:
            Left := LeftAxis
            Up := 0.0
            Forward := ForwardAxis

        Direction.Normalize()


    # Custom normalize function due to built-in `MakeUnitVector[]` returning NaN if magnitude is
    # not greater than 0.
    (V:vector3).Normalize():vector3=
        Magnitude := Sqrt(V.Left * V.Left + V.Up * V.Up + V.Forward * V.Forward)
        if(Magnitude = 0.0):
            return V

        V / Magnitude


    GetRandomColor():color=
        RandomColor := color:
            R := GetRandomFloat(0.0, 1.0)
            G := GetRandomFloat(0.0, 1.0)
            B := GetRandomFloat(0.0, 1.0)            
