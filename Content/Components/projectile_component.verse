<#
@UUID: 0b44618d-1106-4570-971b-fab0a360442c
@Task: ProjectileModule
@Description: A module that handles that handles physics calculations for a projectile.
@Tags: Instantiate Prefab, Physics, Projectile, Bullet, Collision
@Export: true
#>
<#
Entity Prefab setup for the `projectile` entity prefab:
    `projectile` is a top-level entity prefab that gets spawned when a player fires their `tank`.
    `projectile` simulates its own velocity, and checks for overlap hits. At which point, the `projectile`
    explodes and spawns an `explosion` entity prefab that then has the ability to damage opposing players'
    `tank`.

    All components listed in the entity setup hierarchy *must* match to ensure `projectile` acts as expected.

    - projectile
        - mesh_component
            NOTE:
                The mesh associated with mesh_component *must* have collision.
                Its collision *must* allow queries (e.g. `PhysicsActor` collision preset).
                    "Collision Complexity" should be set to `Project Default`.
        - projectile_component
#>
using { /Verse.org }
using { /Verse.org/Native }
using { /Verse.org/SceneGraph }
using { /Verse.org/Simulation }
using { /Verse.org/SpatialMath }
using { Prefabs }

projectile_component := class<final_super>(component):

    @editable
    Gravity: vector3 = vector3:
        Left := 0.0
        Up := -25.0
        Forward := 0.0

    @editable
    Drag: float = 0.02

    # NOTE: Must be set when spawning this entity.
    var Lifetime: float = 0.0
    # NOTE: Must be set when spawning this entity.
    var Velocity: vector3 = vector3{}
    var MaybeTransformComponent: ?transform_component = false
    
    # NOTE: Must be set when spawning this entity.
    var OwnerAgent: ?agent = false

    var MaybePrePhysics: ?cancelable = false


    OnBeginSimulation<override>():void=
        (super:)OnBeginSimulation()

        
        if(TransformComp := Entity.GetComponent[transform_component]):
            set MaybeTransformComponent = option. TransformComp
        else:
            Print("Missing `transform_component`!")

        set MaybePrePhysics = option. TickEvents.PrePhysics.Subscribe(OnPrePhysics)
        

    OnEndSimulation<override>():void=
        if(PrePhysics := MaybePrePhysics?):
            PrePhysics.Cancel()


    OnPrePhysics(Delta:float):void=
        if. Lifetime <= 0.0 then. Entity.RemoveFromParent() else. set Lifetime -= Delta

        if(TransformComp := MaybeTransformComponent?):
            CurrentTransform := Entity.GetGlobalTransform()
            set Velocity = ((Velocity + Gravity * Delta) * (1.0 - (Drag * Delta)))

            NewTransform := transform:
                Rotation := IdentityRotation()
                Translation := CurrentTransform.Translation + Velocity
                Scale := CurrentTransform.Scale

            Entity.SetGlobalTransform(NewTransform)

        if(Entity.GetGlobalTransform().Translation.Up < 0.0):
            Explode()
        else:
            HandleCollisions()


    HandleCollisions():void=
        for:
            Hit: Entity.FindOverlapHits()
            HitTankComp := Hit.TargetComponent.Entity.GetComponent[tank_controller_component]
            not OwnerAgent = HitTankComp.MaybePilot
        do:
            Explode()

            
    Explode():void=
        if(SimRoot := Entity.GetSimulationEntity[]):
            Explosion := explosion{}
            Explosion.SetGlobalTransform(Entity.GetGlobalTransform()) 
            SimRoot.AddEntities(array. Explosion)

            if(ExplosionComp := Explosion.GetComponent[explosion_component]):
                set ExplosionComp.MaybeOwnerAgent = OwnerAgent

        Entity.RemoveFromParent()
