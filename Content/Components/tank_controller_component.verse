<#
@UUID: 7fbfee5f-2e7f-4253-a98c-ccc29def430f
@Task: Custom tank vehicle controller
@Description: Module that manages a custom tank-style vehicle. Players can enter, exit, drive, shoot, and have their tank destroyed. 
@Tags: Vehicle, Pilot, Tank, Driving
@Export: true
#>
<#
Entity Prefab setup for `tank` entity prefab:
    `tank` is a top-level entity prefab that handles state management and player controls of the custom vehicle.
    `turret` is an entity prefab that is a child of `tank` and acts as the spawn point for projectiles.

    All components listed in the entity setup hierarchy *must* match to ensure `tank` acts as expected.

    - tank
        - transform_component
            NOTE: 
                transform_component allows movement and rotation of `tank`.
        - mesh_component
            NOTE: 
                The mesh associated with mesh_component *must* have collision.
                Its collision *must* allow queries (e.g. `PhysicsActor` collision preset).
                    "Collision Complexity" should be set to `Project Default`.
                These settings allow `projectile` and `explosion` prefabs to interact with the `tank`
                    and also prevent players from clipping through the mesh of the `tank`.
        - tank_controller_component
            NOTE:
                tank_controller_component tracks the state of the `tank` and allows a player to
                pilot it (enter, exit, control its movement and rotation, and shoot).
        - turret (entity prefab)
            NOTE: 
                `turret` is an entity prefab and is a child of the `tank` prefab. It contains a method to shoot a projectile while also
                    acting as the spawn point when spawning a projectile.
            - transform_component
            - mesh_component
                NOTE:
                    Mesh for `turret` *must* not have collision or its "Collision Preset" *must*
                    not interact with `projectile` collision otherwise `projectile` would trigger a hit the 
                    moment it is spawned.
            - turret_component
                NOTE:
                    turret_component serves as a spawn point for projectiles and contains a method to spawn projectiles
                        at that point.
            -ui_component
                NOTE:
                    ui_component displays the `tank` health to its pilot tank. 
#>
using { /Verse.org }
using { /Verse.org/Native }
using { /Verse.org/SceneGraph }
using { /Verse.org/Simulation }
using { /Fortnite.com/Devices }
using { /Fortnite.com/Characters }
using { /Verse.org/SpatialMath }
using { /Fortnite.com/FortPlayerUtilities }
using { /Verse.org/Colors }
using { Prefabs }

UESM := import (/UnrealEngine.com/Temporary/SpatialMath)

tank_controller_component := class<final_super>(interactable_component):

    @editable
    MaxHealth: int = 3

    @editable
    MoveSpeed: float = 250.0
    @editable
    RotateSpeed: float = 60.0

    @editable
    TurretOffsetTransform: transform = transform:
        Translation := vector3{Up:=90.0}
        Rotation := MakeRotationFromYawPitchRollDegrees(0.0, 15.0, 0.0)
        Scale := vector3{}

    @editable
    CameraOffset: vector3 = vector3:
        Up := 375.0
        Forward := -300.0

    @editable
    CameraPitch: float = -20.0
    @editable
    CameraYawPredicitionScale: float = 5.0

    @editable
    MinProjectileForce: float = 20.0
    @editable
    MaxProjectileForce: float = 50.0
    @editable
    FireChargeTime: float = 0.75
    @editable
    FireCooldownTime: float = 0.25

    @editable
    PilotExitLocation: vector3 = vector3{Up:=250.0}

    UpdatedNewTankTransform: event() = event(){}
    UpdatedNewCameraTransform: event() = event(){}
    RemovedFromScene: event() = event(){}

    var MaybeGameManager: ?game_manager_component = false
    var MaybeCamera: ?gameplay_camera_fixed_point_device = false

    var InputResult: input_result = input_result{}

    var MaybeUIComponent: ?ui_component = false
    var MaybeTurret: ?turret = false
    var MaybePilot: ?agent = false
    var MaybePilotChar: ?fort_character = false

    var CurrentHealth: int = 0
    var BaseColor: color = color{}

    var NewTankTransform: transform = transform{}
    var NewCameraTransform: transform = transform{}
    var TankYawDelta: float = 0.0
    var PrevTankYaw: float = 0.0

    var ProjectileForce: float = 0.0
    var ChargeDelta: float = 0.0
    var FireCooldownTimer: float = 0.0

    var MaybePrePhysics: ?cancelable = false
    var MaybePostPhysics: ?cancelable = false


    OnBeginSimulation<override>():void=
        (super:)OnBeginSimulation()

        SucceededEvent.Subscribe(OnEnterTank)

        set MaybePrePhysics = option. TickEvents.PrePhysics.Subscribe(OnPrePhysics)
        set MaybePostPhysics = option. TickEvents.PostPhysics.Subscribe(OnPostPhysics)
        
        set CurrentHealth = MaxHealth      
        
        if(UIComp := Entity.GetComponent[ui_component]):
            set MaybeUIComponent = option. UIComp


    OnSimulate<override>()<suspends>:void=
        race:
            RemovedFromScene.Await()

            sync:
                loop:
                    UpdatedNewTankTransform.Await()                        

                    Entity.SetGlobalTransform(NewTankTransform)
                    if:
                        PilotChar := MaybePilotChar?
                        PilotLoc := UESM.FromVector3(NewTankTransform.Translation + PilotExitLocation)
                        PilotRot := UESM.FromRotation(NewTankTransform.Rotation)
                        PilotChar.TeleportTo[PilotLoc, PilotRot]
                    else:
                        Print("Align player failed!")

                    Sleep(0.0)

                loop:
                    if(Turret := MaybeTurret?):
                        TankRot := NewTankTransform.Rotation.GetYawPitchRollDegrees()
                        TurretRot := TurretOffsetTransform.Rotation.GetYawPitchRollDegrees()
                        NewTurretTransform := transform:
                            Translation := NewTankTransform.Translation + TurretOffsetTransform.Translation
                            Rotation := MakeRotationFromYawPitchRollDegrees(TankRot(0) + TurretRot(0), TankRot(1) + TurretRot(1), TankRot(2) + TurretRot(2))
                            Scale := vector3{Left:= 1.0, Up:= 1.0, Forward:= 1.0}
                        Turret.SetGlobalTransform(NewTurretTransform)

                    Sleep(0.0)

                loop:
                    UpdatedNewCameraTransform.Await()

                    if:
                        Camera := MaybeCamera?           
                        CameraLoc := UESM.FromVector3(NewCameraTransform.Translation)
                        CameraRot := UESM.FromRotation(NewCameraTransform.Rotation)                 
                        Camera.TeleportTo[CameraLoc, CameraRot]
                    else:

                    Sleep(0.0)


    OnPrePhysics(Delta:float):void=
        if:
            GameManager := MaybeGameManager?
            Pilot := MaybePilot?
            CurrentTransform := Entity.GetGlobalTransform()
            CurrentYaw := CurrentTransform.Rotation.GetYawPitchRollDegrees()(0)
        then:
            set InputResult = GameManager.GetInput(Pilot)
            MoveDir := InputResult.GetMovementDirection()                
            RelativeFwd := CurrentTransform.Rotation.GetForwardAxis()
            NewRotation := MakeRotationFromYawPitchRollDegrees(CurrentYaw + MoveDir.Left * RotateSpeed * Delta, 0.0, 0.0)

            set NewTankTransform = transform:
                Translation := CurrentTransform.Translation + RelativeFwd * MoveDir.Forward * MoveSpeed * Delta
                Rotation := NewRotation

            UpdatedNewTankTransform.Signal()


    OnPostPhysics(Delta:float):void=
        if:
            Camera := MaybeCamera?
            TankYaw := NewTankTransform.Rotation.GetYawPitchRollDegrees()(0)
            set TankYawDelta = TankYaw - PrevTankYaw
        then:
            LocalFwd := NewTankTransform.Rotation.GetForwardAxis()
            LocalUp := NewTankTransform.Rotation.GetUpAxis()
            CameraTargetRotation := MakeRotationFromYawPitchRollDegrees(TankYaw + (TankYawDelta * CameraYawPredicitionScale), CameraPitch, 0.0)
            TargetLocation := NewTankTransform.Translation + LocalFwd * CameraOffset.Forward + LocalUp * CameraOffset.Up

            CurrentRotation := UESM.FromRotation(Camera.GetTransform().Rotation)
            NewRotation := Slerp(CurrentRotation, CameraTargetRotation, 8.0 * Delta)
            set NewCameraTransform = transform:
                Translation := Lerp(UESM.FromVector3(Camera.GetTransform().Translation), TargetLocation, 4.0 * Delta)
                Rotation := NewRotation

            set PrevTankYaw = TankYaw
            UpdatedNewCameraTransform.Signal()

        if:
            Pilot := MaybePilot?
            InputResult.FirePressed?
        then:
            set FireCooldownTimer -= Delta
            if(ChargeDelta < 1.0 and FireCooldownTimer <= 0.0):
                set ChargeDelta += (Delta / FireChargeTime)
                set ChargeDelta = Clamp(ChargeDelta, 0.0, 1.0)
                set ProjectileForce = Lerp(MinProjectileForce, MaxProjectileForce, ChargeDelta)
        else:
            if:
                Turret := MaybeTurret?
                TurretComp := Turret.GetComponent[turret_component]
                ProjectileForce > 0.0
            then:
                TurretComp.FireProjectile(ProjectileForce, MaybePilot)

                set FireCooldownTimer = FireCooldownTime
                set ChargeDelta = 0.0
                set ProjectileForce = 0.0


    OnRemovingFromScene<override>():void=
        if(Tank := tank[Entity]):
            Tank.EjectPilot()

        if(PrePhysics := MaybePrePhysics?):
            PrePhysics.Cancel()

        if(PostPhysics := MaybePostPhysics?):
            PostPhysics.Cancel()

        RemovedFromScene.Signal()


    OnEnterTank(Agent:agent):void=
        if(Tank := tank[Entity]):
            Tank.EnterAsPilot(Agent)

    
    
    InteractMessage<override>(Agent:agent)<reads><decides>:message= LocText<localizes>:message="Enter Tank"

    AgentNameText<localizes>(Agent:agent):message= "{Agent}"
    ToString(Agent:agent):string= Localize(AgentNameText(Agent))



(Tank:tank).ChangeColor(NewColor:color):void=
    if:
        TankComp := Tank.GetComponent[tank_controller_component]
        TankMesh := Meshes.Tank[Tank.GetComponent[mesh_component]]
    then:
        NewMat := Materials.tank_material{}
        set TankComp.BaseColor = NewColor
        set NewMat.BaseColor = NewColor

        set TankMesh.Material = NewMat


(Tank:tank).EnterAsPilot(Agent:agent):void=
    if:
        TankComp := Tank.GetComponent[tank_controller_component] 
        FortChar := Agent.GetFortCharacter[]
        GameManager := TankComp.MaybeGameManager?
        not TankComp.MaybePilot?
        Camera := TankComp.MaybeCamera?
        UIComp := TankComp.MaybeUIComponent?
    then:           
        set TankComp.MaybePilot = option. Agent
        set TankComp.MaybePilotChar = option. FortChar     
        FortChar.PutInStasis(stasis_args{AllowFalling:=false, AllowTurning:=false, AllowEmotes:=false})
        FortChar.Hide()
        Camera.AddTo(Agent)
        UIComp.AddUI(Agent)


(Tank:tank).EjectPilot():void=
    if:
        TankComp := Tank.GetComponent[tank_controller_component]
        Pilot := TankComp.MaybePilot?
        PilotChar := TankComp.MaybePilotChar?
        PilotLoc := Tank.GetGlobalTransform().Translation + TankComp.PilotExitLocation
        PilotRot := UESM.FromRotation(IdentityRotation())
        PilotChar.TeleportTo[UESM.FromVector3(PilotLoc), PilotRot]
        Camera := TankComp.MaybeCamera?
        UIComp := TankComp.MaybeUIComponent?
    then:
        set TankComp.MaybePilot = false
        set TankComp.MaybePilotChar = false
        Camera.RemoveFrom(Pilot)
        UIComp.RemoveUI(Pilot)
        PilotChar.ReleaseFromStasis()
        PilotChar.Show()
    

(Tank:tank).Damage(Amount:int):void=
    if:
        TankComp := Tank.GetComponent[tank_controller_component]
        Pilot := TankComp.MaybePilot?
        UIComp := TankComp.MaybeUIComponent?

    then:
        set TankComp.CurrentHealth -= Amount
        UIComp.LowerHealth(Pilot, TankComp.CurrentHealth)
        if(TankComp.CurrentHealth <= 0):
            Tank.Destroy()

        
(Tank:tank).Destroy():void=
    Tank.EjectPilot()
    if(Turret := Tank.GetComponent[tank_controller_component].MaybeTurret?):
        Turret.Destroy()

    if(GameManager := Tank.GetComponent[tank_controller_component].MaybeGameManager?):
        GameManager.TankDestroyed.Signal(Tank)

    Tank.RemoveFromParent()
