<#
@UUID: ee95c963-e3b2-4e3f-9dc3-94dd7aa19dda
@Task: Turret controller for `tank` vehicle.
@Description: A module which allows the player to shoot a projectile.
@Tags: Instantiation, Projectile, Turret, Spawning
@Export: true
#>
<#
Entity Prefab setup for `turret` entity prefab:
    `turret` is a top-level entity prefab that handles the spawning of projectiles for a `tank` custom vehicle.

    All components listed in the entity setup hierarchy *must* match to ensure `turret` acts as expected.

    - turret
        NOTE:
            `turret` is an entity prefab itself, but must be made a child of the `tank` entity prefab.
            `turret` acts as the "barrel" of the tank.
        - transform_component
        - turret_component            
#>
using { /Verse.org }
using { /Verse.org/Native }
using { /Verse.org/SceneGraph }
using { /Verse.org/Simulation }
using { /Verse.org/SpatialMath }
using { /UnrealEngine.com/Temporary/SpatialMath }
using { Prefabs }

turret_component := class<final_super>(component):

    @editable
    MuzzleOffset: float = 140.0


    FireProjectile(Force:float, Pilot:?agent):void=
        if(SimRoot := Entity.GetSimulationEntity[]):
            Location := Entity.GetGlobalTransform().Translation
            RelativeFwd := FromVector3(FromRotation(Entity.GetGlobalTransform().Rotation).GetLocalForward()).Normalize()
            SpawnLocation := Location + (RelativeFwd * MuzzleOffset)

            Projectile := projectile{}
            SimRoot.AddEntities(array. Projectile)
            if(ProjectileComp := Projectile.GetComponent[projectile_component]):
                set ProjectileComp.Lifetime = 10.0
                set ProjectileComp.Velocity = RelativeFwd * Force
                set ProjectileComp.OwnerAgent = Pilot

            Projectile.SetGlobalTransform((/Verse.org/SpatialMath:)transform{Translation:=SpawnLocation})

    
(Turret:turret).Destroy():void=
    Turret.RemoveFromParent()
    