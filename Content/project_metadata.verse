<#
@UUID: 1b399740-c5c9-477b-bf4b-abac9b7377e4
@Task: Tank Brothers
@Description: Tank Brothers is a competitive multiplayer minigame where players pilot custom tank-style vehicles that can shoot exploding projectiles. If a tank is hit by an exploding projectile, it loses health. If a tank's health is reduced to 0, it is destroyed. The player piloting the last undestroyed tank is declared the winner.
@Tags: Multiplayer, Competitive, Combat, Vehicle, Projectiles, Explosions, Keyframed Movement, Custom Vehicle.
@Export: true
#>
