<#
@UUID: 0f83d28b-8363-4bc3-96a4-adb8f1cea836
@Task: Custom UI that visually displays player stats.
@Description: Component that displays certain player stats, allowting players to visually track prohress
@Tags: UI, Stats, Components
@Export: true
#>
<#
    ui_component allows the player to visually see stats in the game. Stats can be updated within the canvas of the ui class as well as the look of each widget added.
    The ui_componnet is then discoevered by other entites and is activated through their logic, allowing total control from the entity what to display and when to remove the
    ui.

    `Heart` is a custom texture.
#>
using { /Verse.org }
using { /Verse.org/Native }
using { /Verse.org/SceneGraph }
using { /Verse.org/Simulation }
using { /Verse.org/Assets }
using { /Verse.org/Colors }
using { /UnrealEngine.com/Temporary/SpatialMath }
using { /UnrealEngine.com/Temporary/UI }
using { /Fortnite.com/Devices }
using { CustomImages }

tank_ui:= class():

    var MainCanvas : canvas = canvas{}

    var Heart0 : texture_block = texture_block:
        DefaultImage := Heart
        DefaultTint := NamedColors.Red

    var Heart1 : texture_block = texture_block:
        DefaultImage := Heart
        DefaultTint := NamedColors.Red
    
    var Heart2 : texture_block = texture_block:
        DefaultImage := Heart
        DefaultTint := NamedColors.Red


    HideUI():void=
        MainCanvas.SetVisibility(widget_visibility.Hidden)

    ShowUI():void=
        MainCanvas.SetVisibility(widget_visibility.Visible)


    CreateUI()<transacts>:canvas=
        CreatedUI : canvas = canvas{
            Slots := array:
                canvas_slot:
                    Anchors:= anchors{Minimum := vector2{X:= 0.0, Y:= 0.0}, Maximum := vector2{X:= 0.0, Y:= 0.0}}
                    Offsets := margin{Left:= 544.0, Top:= 760.0, Right:= 188.977875, Bottom:= 171.242355}
                    Alignment := vector2{X:= 0.0, Y:= 0.0}
                    ZOrder := 0
                    SizeToContent := false
                    Widget := Heart0
                canvas_slot:
                    Anchors:= anchors{Minimum := vector2{X:= 0.0, Y:= 0.0}, Maximum := vector2{X:= 0.0, Y:= 0.0}}
                    Offsets := margin{Left:= 868.0, Top:= 760.0, Right:= 188.977875, Bottom:= 171.242355}
                    Alignment := vector2{X:= 0.0, Y:= 0.0}
                    ZOrder := 0
                    SizeToContent := false
                    Widget := Heart1
                canvas_slot:
                    Anchors:= anchors{Minimum := vector2{X:= 0.0, Y:= 0.0}, Maximum := vector2{X:= 0.0, Y:= 0.0}}
                    Offsets := margin{Left:= 1188.0, Top:= 760.0, Right:= 188.977875, Bottom:= 171.242355}
                    Alignment := vector2{X:= 0.0, Y:= 0.0}
                    ZOrder := 0
                    SizeToContent := false
                    Widget := Heart2
        } 
    
        set MainCanvas = CreatedUI


ui_component := class<final_super>(component):

    var MaybeTankHUD: ?tank_ui = false


    AddUI(Agent:agent):void= 
        if:
            Player := player[Agent] 
            PlayerUI := GetPlayerUI[Player]
        then:
            if(not MaybeTankHUD?):
                set MaybeTankHUD = option. tank_ui{}

            if(TankHUD := MaybeTankHUD?):
                ResetHealth(TankHUD)
                PlayerUI.AddWidget(TankHUD.CreateUI())


    RemoveUI(Agent:agent):void=
        if:
            Player := player[Agent]
            PlayerUI := GetPlayerUI[Player]
            TankHUD := MaybeTankHUD?
        then:
            PlayerUI.RemoveWidget(TankHUD.MainCanvas)


    LowerHealth(Agent:agent, Health:int):void=
        if:
            Player := player[Agent]
            PlayerUI := GetPlayerUI[Player]
            TankHUD := MaybeTankHUD?
        then:
            case (Health):
                2 =>
                    TankHUD.Heart2.SetTint(NamedColors.Black)
                1 =>
                    TankHUD.Heart1.SetTint(NamedColors.Black)
                0 =>
                    TankHUD.Heart0.SetTint(NamedColors.Black)
                _ =>
                    Print("Error changing health!")


    ResetHealth(TankHUD:tank_ui):void=
        TankHUD.Heart2.SetTint(NamedColors.Red)
        TankHUD.Heart1.SetTint(NamedColors.Red)
        TankHUD.Heart0.SetTint(NamedColors.Red)
